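# Idle Niagara effect for the ElectroShock mic stand pickaxe cosmetic, layering electrical effects around system-defined sphere and two torus anchor heights.
# CentralColumn spawns rising, flickering meshes between the torus heights, while Spikes emits short-lived velocity-aligned meshes at the sphere or lower torus anchor.
# SphereLightning001 fires looping bursts that form ribbon arcs from torus A to the sphere, using LoopCount as ribbon ID and random seed, with glow and endpoint sprites.
# Animated dynamic material parameters and color curves drive flicker and fade.
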NiagaraSystem /SparksCosmetics/Instrument/MicStand/ElectroShock/Effects/NS_MicStand_Pickaxe_Electroshock_Idle
dependencies {
 usedDataInterfaces [&NiagaraDataInterfaceCurve, &NiagaraDataInterfaceArrayFloat, &NiagaraDataInterfaceColorCurve]
 usedDynamicInputs []
 usedModules [
  &self:SystemSpawnScript.SystemScriptSource.SystemScriptGraph.NiagaraNodeAssignment_0.SetVariables_#1
  &SystemState
  &self:SystemSpawnScript.SystemScriptSource.SystemScriptGraph.NiagaraNodeAssignment_4.SetVariables_#2
  &self:SystemSpawnScript.SystemScriptSource.SystemScriptGraph.NiagaraNodeAssignment_6.SetVariables_#3
  &self:SystemSpawnScript.SystemScriptSource.SystemScriptGraph.NiagaraNodeAssignment_1.SetVariables_#4
 ]
}
emitters {
 CentralColumn {
  emitterUpdateScript [
   &EmitterState()
   &SpawnRate(
    SpawnRate=&RandomRangeFloat(Minimum=8, Maximum=20, Evaluation Type=::NewEnumerator1)
    Use Spawn Probability=false
   )
   &SpawnRate(
    SpawnGroup=1
    SpawnRate=&RandomRangeFloat(Minimum=6, Maximum=10, Evaluation Type=::NewEnumerator1)
    Use Spawn Probability=false
   ) {module=SpawnRate001}
   &SpawnRate(
    SpawnGroup=2
    SpawnRate=&RandomRangeFloat(Minimum=10, Maximum=20, Evaluation Type=::NewEnumerator1)
    Use Spawn Probability=false
   ) {module=SpawnRate002}
  ]
  particleSpawnScript [
   /SetVariables(Particles.SectionPicker=1) {isSetParameters=true}
   /SetVariables(Particles.SectionPicker:NiagaraFloat=&UniformAOrBFloat(A=0, Evaluation Type=::NewEnumerator0, Recalculate Random Each Loop=true, Override Seed=false)) {enabled=false, isSetParameters=true}
   &InitializeParticle(
    Color Maximum=<rgba 0.18 0.18 0.18 1>
    Color Minimum=<rgba 0.075 0.075 0.075 0.7>
    Color Mode=::NewEnumerator2
    Lifetime Max=0.5
    Lifetime Min:NiagaraFloat=&Lerp_Float(A=0.3, B=0.1, Alpha=$Particles.SectionPicker)
    Lifetime Mode=::NewEnumerator1
    Mesh Scale Mode=::NewEnumerator3
    Mesh Uniform Scale=&Multiply_Float(
     A=&Lerp_Float(A=1.5, B=1, Alpha=$Particles.SectionPicker)
     B=&MaskFloatBySpawnGroup(
      Masked Float=&RandomRangeFloat(Minimum=0.075, Maximum=0.125)
      Passthrough Float=&MaskFloatBySpawnGroup(
       Masked Float=0.25
       Passthrough Float=&RandomRangeFloat(Minimum=0.075, Maximum=0.125)
       Spawn Group Mask=1
      )
     )
    )
    Position Offset=&Lerp_Vector(
     A=&Add_Vector(A=<0 0 -2>, B=$System.SphereHeight)
     B=&Add_Vector(A=<0 0 -2>, B=&RandomRangeVector(Minimum=$System.TorusBHeight, Maximum=$System.TorusAHeight))
     Alpha=$Particles.SectionPicker
    )
    Sprite Size Mode=::NewEnumerator3
    UsePositionOffset=true
    Write Material Random=true
   )
   &InitialMeshOrientation(
    Mesh Orientation Mode=::NewEnumerator2
    Rotation=&RandomRangeVector(Minimum=<0 0 0>, Maximum=<0 0 1>)
    Use Rotation Vector=true
   )
   &CameraOffset(Camera Offset Amount=&Add_Float(
     A=&MaskFloatBySpawnGroup(Masked Float=2, Passthrough Float=1)
     B=&Lerp_Float(A=5, B=0, Alpha=$Particles.SectionPicker)
    ))
   &AddVelocity(
    Velocity=&RandomRangeVector(Minimum=<0 0 10>, Maximum=<0 0 40>)
    Velocity Speed Scale=&Lerp_Float(A=1.2, B=1, Alpha=$Particles.SectionPicker)
   )
  ]
  particleUpdateScript [
   &ParticleState()
   &UpdateMeshOrientation(
    Rotation Rate=&RandomRangeFloat(Minimum=0, Maximum=1)
    Rotation Vector=&RandomVector(Vector Scale=1)
   )
   &ScaleMeshSize(Scale Factor=&VectorFromFloat(Value=&FloatFromCurve(FloatCurve=DataInterface() {propertyValues {Curve {keys [
          {arriveTangent=1.9115, interpMode=RCIM_Cubic, leaveTangent=1.9115, tangentMode=RCTM_User, time=0, value=0}
          {arriveTangent=0.0224, interpMode=RCIM_Cubic, leaveTangent=0.0224, tangentMode=RCTM_User, time=1, value=1}
         ]}}})))
   &DynamicMaterialParameters(
    Index 0 Param 4=13
    Index 1 Param 1=&FloatFromCurve(FloatCurve=DataInterface() {propertyValues {Curve {keys [<at 0 0.5>,<at 1 1>]}}})
    Index 1 Param 2=2
    Index 1 Param 3=&FloatFromCurve(FloatCurve=DataInterface() {propertyValues {Curve {keys [{interpMode=RCIM_Cubic, tangentMode=RCTM_Break, time=0.5, value=-0.5}, <at 1 0.1>]}}})
    Index 2 Param 1=&Multiply_Float(
     A=&FloatFromCurve(
      FloatCurve=DataInterface() {propertyValues {Curve {keys [{interpMode=RCIM_Cubic, leaveTangent=8.1085, tangentMode=RCTM_Break, time=0.8568, value=1}, {tangentMode=RCTM_User, time=1, value=0}]}}}
      Scale Curve=0.25
     )
     B=&Divide_Float(
      A=&MakeFloatFromVector(Vector=$Particles.Scale, Channel=::NewEnumerator0)
      B=&RandomRangeFloat(Minimum=10, Maximum=15)
     )
    )
    Index 2 Param 2=&FloatFromCurve(
     FloatCurve=DataInterface() {propertyValues {Curve {keys [
         {interpMode=RCIM_Cubic, tangentMode=RCTM_Break, time=0.7, value=0}
         {arriveTangent=-0.0859, interpMode=RCIM_Cubic, tangentMode=RCTM_Break, time=1, value=-1}
        ]}}}
     Scale Curve=0
    )
    Index 2 Param 3=&Waveform(
     Global Input Position=&RandomRangeFloat(Minimum=0, Maximum=15)
     Global Phase=$Particles.Age
     "Waveform [1]"=::NewEnumerator12
     "[1] Phase Scale"=2
     "[1] Frequency"=1
     "[1] Amplitude Min/Max"=<0 1>
    )
    Index 2 Param 4=25
    Index 3 Param 1=&FloatFromCurve(
     FloatCurve=DataInterface() {propertyValues {Curve {keys [
         {interpMode=RCIM_Cubic, leaveTangent=-0.0025, tangentMode=RCTM_Break, time=0, value=0}
         {arriveTangent=1.8528, interpMode=RCIM_Cubic, tangentMode=RCTM_Break, time=1, value=1}
        ]}}}
     Scale Curve=&RandomRangeFloat(Minimum=0, Maximum=0)
    )
    Index 3 Param 2=&FloatFromCurve(
     FloatCurve=DataInterface() {propertyValues {Curve {keys [
         {interpMode=RCIM_Cubic, leaveTangent=-0.0025, tangentMode=RCTM_Break, time=0, value=0.5}
         {arriveTangent=-0.0316, interpMode=RCIM_Cubic, tangentMode=RCTM_Break, time=1, value=-0.2}
        ]}}}
     Scale Curve=0
    )
    Index 3 Param 3=&FloatFromCurve(FloatCurve=DataInterface() {propertyValues {Curve {keys [
         {interpMode=RCIM_Cubic, leaveTangent=1.1034, tangentMode=RCTM_Break, time=0, value=-1}
         {arriveTangent=2.2399, tangentMode=RCTM_User, time=0.5, value=0.125}
        ]}}})
    Index 3 Param 4=&FloatFromCurve(FloatCurve=DataInterface() {propertyValues {Curve {keys [
         {interpMode=RCIM_Cubic, leaveTangent=1.8089, tangentMode=RCTM_Break, time=0.3, value=-0.5}
         {arriveTangent=2.0757, interpMode=RCIM_Cubic, tangentMode=RCTM_Break, time=0.9, value=1}
        ]}}})
    Module.Index0Param0Write=true
    Module.Index0Param1Write=true
    Module.Index0Param2Write=true
    Module.Index0Param3Write=true
    Module.Index1Param0Write=true
    Module.Index1Param1Write=true
    Module.Index1Param2Write=true
    Module.Index1Param3Write=true
    Module.Index2Param0Write=true
    Module.Index2Param1Write=true
    Module.Index2Param2Write=true
    Module.Index2Param3Write=true
    Module.Index3Param0Write=true
    Module.Index3Param1Write=true
    Module.Index3Param2Write=true
    Module.Index3Param3Write=true
    Write Parameter Index 1=true
    Write Parameter Index 2=true
    Write Parameter Index 3=true
   )
   &ScaleColor(
    Scale Alpha=&FloatFromCurve(FloatCurve=DataInterface() {propertyValues {Curve {keys [
         {tangentMode=RCTM_User, time=0, value=1}
         {arriveTangent=-4.3986, interpMode=RCIM_Cubic, leaveTangent=-4.3986, tangentMode=RCTM_User, time=0.35, value=1}
         {arriveTangent=0.1205, interpMode=RCIM_Cubic, leaveTangent=0.1205, tangentMode=RCTM_User, time=1, value=0}
        ]}}})
    Scale RGB=&VectorFromFloat(Value=&Lerp_Float(
      A=20
      B=&RandomRangeFloat(Minimum=-0.5, Maximum=2, Evaluation Type=::NewEnumerator1)
      Alpha=&FloatFromCurve(FloatCurve=DataInterface() {propertyValues {Curve {keys [
           {arriveTangent=1.0951, interpMode=RCIM_Cubic, leaveTangent=1.0951, leaveTangentWeight=0.3923, tangentMode=RCTM_User, time=0, value=0}
           {arriveTangentWeight=0.3923, interpMode=RCIM_Cubic, tangentMode=RCTM_User, time=1, value=1}
          ]}}})
     ))
    ScaleA=true
    ScaleRGB=true
   )
   /SetVariables(Particles.Glow Mesh Color:LinearColor=&MultiplyLinearColor(
     Linear Color=$Particles.Color
     Scale Factor=&MakeFloat4FromVectorAndFloat("Vector (RGB)"=<1 1 1>, "Float (Alpha)"=5)
    )) {isSetParameters=true}
   &Drag(Drag=2, Use Linear Drag=true, Use Rotational Drag=true)
   &SolveForcesAndVelocity(Clamp Velocity=false, Limit Acceleration=false)
  ]
  rendererClasses [&NiagaraMeshRendererProperties]
  renderers [
   {rendererClass=&NiagaraMeshRendererProperties, rendererIndex=0}
   {rendererClass=&NiagaraMeshRendererProperties, rendererIndex=1}
  ]
 }
 SphereLightning001 {
  emitterUpdateScript [
   &EmitterState(
    Life Cycle Mode=::NewEnumerator1
    Loop Delay=&RandomRangeFloat(Minimum=0, Maximum=0.15, Evaluation Type=::NewEnumerator1)
    Loop Duration=&RandomRangeFloat(Minimum=0, Maximum=0.05, Evaluation Type=::NewEnumerator1)
    UseLoopDelay=true
   )
   &SpawnBurst_Instantaneous(
    Spawn Count=&RandomRangeInt(Minimum=5, Maximum=10, Evaluation Type=::NewEnumerator1)
    Use Spawn Probability=false
    UseLoopCountLimit=false
   )
  ]
  particleSpawnScript [
   /SetVariables(
    Particles.RibbonID=/Niagara/DynamicInputs/NiagaraID/MakeNiagaraIDFromInt(Index:NiagaraInt32=$Emitter.LoopCount, Acquire Tag:NiagaraInt32=0)
    Particles.RibbonOrder:NiagaraFloat=&ReturnNormalizedExecIndex()
    Transient.TopOffset:Vector3f=&RandomVector(Vector Scale=5, Randomness Mode=::NewEnumerator1, Random Seed=$Emitter.LoopCount, Fixed Random Seed=true)
   ) {isSetParameters=true}
   /SetVariables(
    Transient.SphereOffset:Vector3f=&Add_Vector(A=$System.SphereHeight, B=$Transient.TopOffset)
    Transient.TopCenterpoint:Vector3f=&Add_Vector(
     A=&Multiply_Vector(A=<1.5 1.5 1>, B=$Transient.TopOffset)
     B=&Lerp_Vector(A=$System.SphereHeight, B=$System.TorusAHeight, Alpha=0.5)
    )
    Transient.TorusAOffset:Vector3f=&Add_Vector(A=$System.TorusAHeight, B=&Multiply_Vector(A=<1.3 1.3 0.5>, B=$Transient.TopOffset))
   ) {isSetParameters=true}
   /SetVariables(Particles.Random=&RandomRangeFloat(Minimum=0, Maximum=1, Randomness Mode=::NewEnumerator1, Random Seed=$Emitter.LoopCount, Fixed Random Seed=true)) {isSetParameters=true}
   &InitializeParticle(
    Color=&ColorFromCurve(
     ColorCurve=DataInterface() {propertyValues {
       AlphaCurve {keys [<at 0 1 RCIM_Cubic>]}
       BlueCurve {keys [<at 0 1 RCIM_Cubic>,<at 0.5 1>,<at 1 1 RCIM_Cubic>]}
       GreenCurve {keys [<at 0 0.58 RCIM_Cubic>,<at 0.5 1>,<at 1 0.58 RCIM_Cubic>]}
       RedCurve {keys [<at 0 0.5 RCIM_Cubic>,<at 0.5 1>,<at 1 0.5 RCIM_Cubic>]}
      }}
     CurveIndex=$Particles.RibbonOrder
    )
    Lifetime=&RandomRangeFloat(Minimum=0.2, Maximum=0.4, Randomness Mode=::NewEnumerator1, Random Seed=$Emitter.LoopCount, Fixed Random Seed=true)
    Position Offset=/Niagara/DynamicInputs/Math/LerpMultipleVectors(
     Alpha:NiagaraFloat=$Particles.RibbonOrder
     V 1:Vector3f=$Transient.TorusAOffset
     V 2:Vector3f=$Transient.TopCenterpoint
     V 3:Vector3f=$Transient.SphereOffset
     V 4:Vector3f=?
     V 5:Vector3f=?
     Vector Count:ENiagara_MultipleLerpCount=::NewEnumerator0
    )
    Position Offset Coordinate Space=::Local
    Ribbon Width=&Multiply_Float(
     A=&RandomRangeFloat(Minimum=2, Maximum=7, Randomness Mode=::NewEnumerator1, Random Seed=$Emitter.LoopCount, Fixed Random Seed=true)
     B=/Niagara/DynamicInputs/Math/LerpMultipleFloats(
      Float Count:ENiagara_MultipleLerpCount=::NewEnumerator0
      1:NiagaraFloat=0.2
      2:NiagaraFloat=1
      3:NiagaraFloat=0.2
      4:NiagaraFloat=?
      5:NiagaraFloat=?
      Alpha:NiagaraFloat=$Particles.RibbonOrder
     )
    )
    Ribbon Width Mode=::NewEnumerator1
    Sprite Size Mode=::NewEnumerator3
    Uniform Sprite Size=5
    UsePositionOffset=true
    Write Material Random=true
   )
   &AddVelocity(Velocity=&RandomVector(Vector Scale=6)) {module=AddVelocity001}
   &AddVelocity(
    Module.Use Falloff Distance=false
    Override Seed=false
    Velocity Mode=::NewEnumerator1
    Velocity Origin=&Lerp_Vector(A=$System.SphereHeight, B=$System.TorusAHeight, Alpha=0.5)
    Velocity Speed=&LerpMultipleFloats(
     Float Count:ENiagara_MultipleLerpCount=::NewEnumerator0
     1:NiagaraFloat=0
     2:NiagaraFloat=20
     3:NiagaraFloat=0
     4:NiagaraFloat=?
     5:NiagaraFloat=?
     Alpha:NiagaraFloat=$Particles.RibbonOrder
    )
   ) {module=AddVelocity002}
  ]
  particleUpdateScript [
   &ParticleState()
   &Drag(Drag=3, Use Linear Drag=true, Use Rotational Drag=true)
   &SolveForcesAndVelocity(Clamp Velocity=false, Limit Acceleration=false)
   &ScaleColor(
    Linear Color Curve=DataInterface() {propertyValues {
      AlphaCurve {keys [<at 0 1>,<at 1 0>]}
      BlueCurve {keys [<at 0 2>,<at 1 1>]}
      GreenCurve {keys [<at 0 2>,<at 1 0.5>]}
      RedCurve {keys [<at 0 2>,<at 1 0.5765>]}
     }}
    Scale Mode=::NewEnumerator2
   )
   &ScaleRibbonWidth(Ribbon Width Scale=&Lerp_Float(A=1, B=&RandomRangeFloat(Minimum=0, Maximum=2), Alpha=$Particles.NormalizedAge))
   &DynamicMaterialParameters(
    Index 1 Param 1=$Particles.Random
    Index 1 Param 2=1
    Index 1 Param 3=&FloatFromCurve(
     FloatCurve=DataInterface() {propertyValues {Curve {keys [<at 0 0>,<at 0.5 1>,<at 0.6 1>,<at 1 0>]}}}
     CurveIndex=$Particles.RibbonOrder
     Scale Curve=0
    )
    Index 1 Param 4=&FloatFromCurve(FloatCurve=DataInterface() {propertyValues {Curve {
        defaultValue=0
        keys [<at 0 0 RCIM_Cubic>, {arriveTangent=0.7712, interpMode=RCIM_Cubic, leaveTangent=0.7712, tangentMode=RCTM_User, time=1, value=1}]
       }}})
    Module.Index1Param0Write=true
    Module.Index1Param1Write=true
    Module.Index1Param2Write=true
    Module.Index1Param3Write=true
    Write Parameter Index 0=false
    Write Parameter Index 1=true
   )
   /SetVariables(
    Particles.GlowColor=&MultiplyLinearColor(
     Linear Color=$Particles.Color
     Scale Factor=&MakeFloat4FromVectorAndFloat(
      "Vector (RGB)"=&VectorFromFloat(Value=&FloatFromCurve(FloatCurve=DataInterface() {propertyValues {Curve {keys [<at 0.7 1>,<at 1 5>]}}}, Scale Curve=1.5))
      "Float (Alpha)"=1
     )
    )
    Particles.GlowDynamics:Vector4f=&MakeVector4(
     X=&MakeFloatFromVector4(Vector 4=$Particles.DynamicMaterialParameter1, Channel=::NewEnumerator0)
     Y=&MakeFloatFromVector4(Vector 4=$Particles.DynamicMaterialParameter1, Channel=::NewEnumerator1)
     Z=&MakeFloatFromVector4(Vector 4=$Particles.DynamicMaterialParameter1, Channel=::NewEnumerator2)
     W=&ClampFloat(Float=&Add_Float(
       A=-0.5
       B=&MakeFloatFromVector4(Vector 4=$Particles.DynamicMaterialParameter1, Channel=::NewEnumerator3)
      ))
    )
    Particles.GlowWidth:NiagaraFloat=&Multiply_Float(A=$Particles.RibbonWidth, B=1.5)
   ) {isSetParameters=true}
   /SetVariables(
    Particles.EndColor:LinearColor=&MultiplyLinearColor(
     Linear Color=&RandomRangeLinearColor(Minimum=<rgba 0.2261 0.621 2 0.5>, Maximum=<rgba 0.3374 1.0545 2 0.5>)
     Scale Factor=&MakeVector4_FromLinearColor(LinearColor=&ColorFromCurve(ColorCurve=DataInterface() {propertyValues {
         AlphaCurve {keys [<at 0 1 RCIM_Cubic>,<at 0.75 1>]}
         BlueCurve {keys [<at 0 1 RCIM_Cubic>,<at 0.75 1 RCIM_Cubic>]}
         GreenCurve {keys [<at 0 0.4296 RCIM_Cubic>,<at 0.75 0.1061 RCIM_Cubic>]}
         RedCurve {keys [<at 0 0.6716 RCIM_Cubic>,<at 0.75 0.4294 RCIM_Cubic>]}
        }}))
    )
    Particles.EndFade:Vector4f=<3 0 0 0>
    Particles.EndOffset=1
    Particles.EndScale:Vector2f=&Multiply_Vector2D_ByFloat(
     Float=&FloatFromCurve(
      FloatCurve=DataInterface() {propertyValues {Curve {keys [<at 0 0.5>,<at 0.01 0>,<at 0.99 0>,<at 1 1>]}}}
      CurveIndex=$Particles.RibbonOrder
      Scale Curve=&FloatFromCurve(FloatCurve=DataInterface() {propertyValues {Curve {keys [<at 0 1>,<at 0.75 0.25>]}}})
     )
     Vector2D=$Particles.SpriteSize
    )
   ) {isSetParameters=true}
  ]
  rendererClasses [&NiagaraSpriteRendererProperties, &NiagaraRibbonRendererProperties]
  renderers [
   {rendererClass=&NiagaraSpriteRendererProperties}
   {rendererClass=&NiagaraRibbonRendererProperties, rendererIndex=1}
   {rendererClass=&NiagaraRibbonRendererProperties, rendererIndex=2}
  ]
 }
 Spikes {
  emitterUpdateScript [
   &EmitterState()
   &SpawnRate(SpawnRate=10, Use Spawn Probability=false)
   &SpawnRate(SpawnGroup=1, SpawnRate=10, Use Spawn Probability=false) {module=SpawnRate001}
   &SpawnRate(
    SpawnGroup=2
    SpawnRate=&RandomRangeFloat(Minimum=2, Maximum=10, Evaluation Type=::NewEnumerator1)
    Use Spawn Probability=false
   ) {module=SpawnRate002}
   &SpawnRate(SpawnGroup=3, SpawnRate=15, Use Spawn Probability=false) {module=SpawnRate003}
  ]
  particleSpawnScript [
   /SetVariables(Particles.SectionPicker:NiagaraFloat=&UniformAOrBFloat(A=0, Distribution Weight=0.75, Evaluation Type=::NewEnumerator0, Recalculate Random Each Loop=true, Override Seed=false)) {isSetParameters=true}
   /SetVariables(Particles.VisibilityTag=&MakeCustomIntFromBool(
     Bool=&SetBoolByIntComparison(A=&SpawnGroup(), B=1, Comparison Type=::NewEnumerator4)
     True INT=1
     False INT=0
    )) {isSetParameters=true}
   &InitializeParticle(
    Color=&RandomRangeLinearColor(Minimum=<rgba 0.8 0.8 0.8 0.2>, Maximum=<rgba 1 1 1 0.5>)
    Lifetime=&MaskFloatBySpawnGroup(
     Masked Float=&RandomRangeFloat(Minimum=0.3, Maximum=0.5)
     Passthrough Float=&MaskFloatBySpawnGroup(
      Masked Float=&RandomRangeFloat(Minimum=0.1, Maximum=0.4)
      Passthrough Float=&RandomRangeFloat(Minimum=0.5, Maximum=0.6)
      Spawn Group Mask=3
     )
    )
    Mesh Scale Mode=::NewEnumerator3
    Mesh Uniform Scale=&MaskFloatBySpawnGroup(
     Masked Float=&RandomRangeFloat(Minimum=0.04, Maximum=0.06)
     Passthrough Float=&RandomRangeFloat(Minimum=0.05, Maximum=0.1)
     Spawn Group Mask=2
    )
    Position Offset=&Lerp_Vector(A=&Add_Vector(A=$System.SphereHeight, B=<0 0 1>), B=$System.TorusBHeight, Alpha=$Particles.SectionPicker)
    UsePositionOffset=true
    Write Material Random=true
   )
   &AddVelocity(Velocity=&Lerp_Vector(
     A=&RandomRangeVector(Minimum=<-1 -1 1>, Maximum=<1 1 1>)
     B=&RandomRangeVector(Minimum=<-5 -5 -1>, Maximum=<5 5 1>)
     Alpha=$Particles.SectionPicker
    ))
   &InitialMeshOrientation(Mesh Axis To Orient=$Particles.Velocity, Mesh Orientation Mode=::NewEnumerator2, Orientation Vector=<0 0 1>, Use Rotation Vector=false)
  ]
  particleUpdateScript [
   &ParticleState()
   &ScaleMeshSize(Scale Factor=&RandomRangeVector(Minimum=<0.5 1 1.5>, Maximum=<1.5 1.3 1.75>)) {module=ScaleMeshSize001}
   &ScaleMeshSize(Scale Factor=&MakeVector(
     X=1
     Y=1
     Z=&FloatFromCurve(
      FloatCurve=DataInterface() {propertyValues {Curve {keys [
          {interpMode=RCIM_Cubic, leaveTangent=0.0966, tangentMode=RCTM_Break, time=0, value=0.93}
          {arriveTangent=0.0951, interpMode=RCIM_Cubic, leaveTangent=-0.1745, tangentMode=RCTM_Break, time=1, value=1}
         ]}}}
      Scale Curve=&Lerp_Float(A=0.5, B=0.75, Alpha=$Particles.SectionPicker)
     )
    ))
   &DynamicMaterialParameters(
    Index 0 Param 4=&MakeCustomFloatFromBool(
     Bool=false
     True Float=5
     False Float=&SelectFloatFromArray(
      Array Sampling Mode=::NewEnumerator0
      Direct Array Index=?
      Float Selection Array=DataInterface() {propertyValues {FloatData [3,0,10,3,1,9,5]}}
      Random Seed=0
     )
    )
    Index 1 Param 1=1
    Index 1 Param 2=2
    Index 1 Param 3=&FloatFromCurve(FloatCurve=DataInterface() {propertyValues {Curve {keys [{interpMode=RCIM_Cubic, tangentMode=RCTM_Break, time=0.5, value=-0.5}, <at 1 0.1>]}}})
    Index 2 Param 1=&Multiply_Float(
     A=&FloatFromCurve(FloatCurve=DataInterface() {propertyValues {Curve {keys [{interpMode=RCIM_Cubic, leaveTangent=8.1085, tangentMode=RCTM_Break, time=0.8568, value=1}, {tangentMode=RCTM_User, time=1, value=0}]}}})
     B=&Multiply_Float(
      A=&MakeFloatFromVector(Vector=$Particles.Scale, Channel=::NewEnumerator0)
      B=&RandomRangeFloat(Minimum=0.05, Maximum=0.075)
     )
    )
    Index 2 Param 2=&FloatFromCurve(
     FloatCurve=DataInterface() {propertyValues {Curve {keys [
         {interpMode=RCIM_Cubic, tangentMode=RCTM_Break, time=0.7, value=0}
         {arriveTangent=-0.0859, interpMode=RCIM_Cubic, tangentMode=RCTM_Break, time=1, value=-1}
        ]}}}
     Scale Curve=50
    )
    Index 2 Param 3=&Waveform(
     Global Input Position=&RandomRangeFloat(Minimum=0, Maximum=15)
     Global Phase=$Particles.Age
     "Waveform [1]"=::NewEnumerator12
     "[1] Phase Scale"=2
     "[1] Frequency"=1
     "[1] Amplitude Min/Max"=<0 1>
    )
    Index 2 Param 4=25
    Index 3 Param 1=&FloatFromCurve(
     FloatCurve=DataInterface() {propertyValues {Curve {keys [
         {interpMode=RCIM_Cubic, leaveTangent=-0.0025, tangentMode=RCTM_Break, time=0, value=0}
         {arriveTangent=1.8528, interpMode=RCIM_Cubic, tangentMode=RCTM_Break, time=1, value=1}
        ]}}}
     Scale Curve=&RandomRangeFloat(Minimum=15, Maximum=25)
    )
    Index 3 Param 2=&FloatFromCurve(FloatCurve=DataInterface() {propertyValues {Curve {keys [
         {interpMode=RCIM_Cubic, leaveTangent=-0.0025, tangentMode=RCTM_Break, time=0, value=0.5}
         {arriveTangent=-0.0316, interpMode=RCIM_Cubic, tangentMode=RCTM_Break, time=1, value=-0.2}
        ]}}})
    Index 3 Param 3=&FloatFromCurve(FloatCurve=DataInterface() {propertyValues {Curve {keys [
         {interpMode=RCIM_Cubic, leaveTangent=1.1034, tangentMode=RCTM_Break, time=0, value=-1}
         {arriveTangent=2.2399, tangentMode=RCTM_User, time=0.5, value=0.125}
        ]}}})
    Index 3 Param 4=&FloatFromCurve(FloatCurve=DataInterface() {propertyValues {Curve {keys [
         {interpMode=RCIM_Cubic, leaveTangent=3.0601, tangentMode=RCTM_Break, time=0.5, value=-0.5}
         {arriveTangent=3.2053, tangentMode=RCTM_User, time=1, value=1}
        ]}}})
    Module.Index0Param0Write=true
    Module.Index0Param1Write=true
    Module.Index0Param2Write=true
    Module.Index0Param3Write=true
    Module.Index1Param0Write=true
    Module.Index1Param1Write=true
    Module.Index1Param2Write=true
    Module.Index1Param3Write=true
    Module.Index2Param0Write=true
    Module.Index2Param1Write=true
    Module.Index2Param2Write=true
    Module.Index2Param3Write=true
    Module.Index3Param0Write=true
    Module.Index3Param1Write=true
    Module.Index3Param2Write=true
    Module.Index3Param3Write=true
    Write Parameter Index 1=true
    Write Parameter Index 2=true
    Write Parameter Index 3=true
   )
   &SolveForcesAndVelocity(Clamp Velocity=false, Limit Acceleration=false)
   &ScaleColor(
    Scale Alpha=&FloatFromCurve(FloatCurve=DataInterface() {propertyValues {Curve {keys [
         {tangentMode=RCTM_User, time=0, value=1}
         {arriveTangent=-0.0405, interpMode=RCIM_Cubic, leaveTangent=-0.0405, tangentMode=RCTM_User, time=0.7009, value=1}
         {arriveTangent=-8.0089, interpMode=RCIM_Cubic, leaveTangent=-8.0089, tangentMode=RCTM_User, time=1, value=0}
        ]}}})
    Scale RGB=&Lerp_Vector(
     A=&VectorFromFloat(Value=&FloatFromCurve(FloatCurve=DataInterface() {propertyValues {Curve {keys [<at 0.2771 1>,<at 0.3786 1>]}}}, Scale Curve=20))
     B=&VectorFromFloat(Value=&RandomRangeFloat(Minimum=0.3, Maximum=2, Evaluation Type=::NewEnumerator1))
     Alpha=&FloatFromCurve(FloatCurve=DataInterface() {propertyValues {Curve {keys [
          {interpMode=RCIM_Cubic, leaveTangent=2.5294, tangentMode=RCTM_Break, time=0, value=0}
          {arriveTangent=-0.2166, interpMode=RCIM_Cubic, tangentMode=RCTM_Break, time=0.4, value=1}
         ]}}})
    )
    ScaleA=true
    ScaleRGB=true
   )
  ]
  rendererClasses [&NiagaraMeshRendererProperties]
  renderers [
   {rendererClass=&NiagaraMeshRendererProperties, rendererIndex=0}
   {rendererClass=&NiagaraMeshRendererProperties, rendererIndex=1}
   {rendererClass=&NiagaraMeshRendererProperties, rendererIndex=2}
  ]
 }
}
summary {
 systemName=NS_MicStand_Pickaxe_Electroshock_Idle
 emitters [
  {emitterName=Spikes, rendererClasses [&NiagaraMeshRendererProperties]}
  {emitterName=CentralColumn, rendererClasses [&NiagaraMeshRendererProperties]}
  {
   emitterName=SphereLightning001
   rendererClasses [&NiagaraSpriteRendererProperties, &NiagaraRibbonRendererProperties]
  }
 ]
 userVariables [
  User.bIsFrontEnd {type=NiagaraBool}
  User.bIsFrontEndPreview {type=NiagaraBool}
  User.IsFrontEnd {type=NiagaraBool}
  User.MusicReactive {type=NiagaraBool, defaultValue=true}
  User.Strum {type=NiagaraFloat, defaultValue=1, description=""}
  User.Variant {type=NiagaraFloat, defaultValue=0, description=""}
 ]
}
